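# Translucent, two-sided unlit material that draws a faint, animated, noise-warped skull on Niagara mesh particles, apparently for River Styx effects.
# A radially mapped, panning swirl texture distorts the skull UVs, while dynamic SkullAnimation slides two red-channel samples from opposite vertical offsets into alignment, adding a green-channel layer.
# The mask is multiplied by inverse Fresnel, particle alpha, a world-height fade from dynamic HEIGHT, a custom-depth occlusion fade and V-edge fades.
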Material /SunRoseEnvGameplay/Assets/FX/Materials/M_RiverStyx_Skull_Anim
graph {
 properties {
  allowTranslucentCustomDepthWrites=true
  bUsedWithNiagaraMeshParticles=true
  blendMode=BLEND_Translucent
  pixelDepthOffsetMode=PDOM_Legacy
  shadingModel=MSM_Unlit
  twoSided=true
  usedShadingModels=Unlit
 }
 nodes [
  Constant() {r=1}
  MaterialFunctionCall!(ObjectOrParticlePosition=ObjectPositionWS()) {materialFunction=/Game/Effects/Fort_Effects/Materials/Toon/CustomDepthOutLine/MF_BackfaceCulling}
  n1 = Power(Base=OneMinus(None=Fresnel() {baseReflectFraction=0, exponent=1})) {constExponent=1.5}
  MaterialFunctionCall(In=@n1, Contrast=ScalarParameter() {defaultValue=0.1, parameterName=COntra}) {materialFunction=&CheapContrast}
  MaterialFunctionCall(Emissive=VectorParameter() {parameterName=Color, defaultValue=<rgba 0.0724 0.1979 0.1362 1>}.RGB) {materialFunction=&MF_FX_EmissiveAdjustments}
  n2 = WorldPosition()
  n3 = DynamicParameter() {paramNames [SkullAnimation, HEIGHT, Param3, Param4]}
  n4 = Divide(A=Subtract(
    A=MaterialFunctionCall(Float3=@n2.XYZ) {materialFunction=&BreakOutFloat3Components}.B
    B=@n3.HEIGHT
   )) {constB=32}
  n5 = MaterialFunctionCall(Float3=SceneTexture() {sceneTextureId=PPI_CustomDepth}.Color) {materialFunction=&BreakOutFloat3Components}
  n6 = PixelDepth()
  n7 = Subtract(A=@n5.R, B=ComponentMask(None=@n6.R) {r=true})
  n8 = ScalarParameter() {defaultValue=20, parameterName=CustomDepthFade}
  n9 = Constant()
  n10 = Constant() {r=1}
  n11 = Constant() {r=1}
  n12 = Multiply(
   A=Saturate(None=Divide(A=@n7, B=@n8) {constB=50})
   B=ShadingPathSwitch(Default=If(A=@n6.R, B=@n5.R, "A > B"=@n9, "A == B"=@n10, "A < B"=@n10) {equalsThreshold=0.1}, Forward=@n11, Mobile=@n11)
  )
  n13 = MaterialFunctionCall(Float4=Constant4Vector() {constant=<rgba 1 1.3 1 1.3>}.RGBA) {materialFunction=&BreakOutFloat4Components}
  n14 = TextureCoordinate()
  n15 = MaterialFunctionCall() {materialFunction=&VectorToRadialValue}
  n16 = VectorParameter() {parameterName=Speed, defaultValue=<rgba 0 -0.5 0.5 2>}
  n17 = AppendVector(A=@n16.B, B=@n16.B)
  n18 = TextureSample(UVs=Panner(
    Coordinate=MaterialFunctionCall(UVs=@n15.Radial Coordinates, Texture Scale=@n17) {materialFunction=&ScaleUVsByCenter}.UVs
    Speed=AppendVector(A=@n16.R, B=@n16.R)
   )) {samplerType=SAMPLERTYPE_Normal, texture=/Game/Gadgets/Assets/VinderTech_GliderChute/Glider_Thrive_Female/FX/Texture/T_SwirlNoise_256}
  n19 = AppendVector(A=@n18.R, B=@n18.R)
  n20 = ScalarParameter() {defaultValue=0.2, parameterName=Noise2}
  n21 = MaterialFunctionCall(Float2=@n15.Radial Coordinates) {materialFunction=&BreakOutFloat2Components}
  n22 = Multiply(
   A=Subtract(A=Multiply(A=@n21.G), B=ScalarParameter() {defaultValue=0.2, parameterName=MaskOffset})
   B=ScalarParameter() {defaultValue=1, parameterName=MaskScale}
  )
  n23 = Add(A=@n14, B=Multiply(A=Multiply(A=@n19, B=@n20), B=Saturate(None=@n22)))
  n24 = Multiply(A=AppendVector(A=@n13.B, B=@n13.B), B=@n23)
  n25 = Constant2Vector() {g=0.1}
  n26 = Saturate(None=@n3.SkullAnimation)
  n27 = TextureSample(UVs=LinearInterpolate(A=Add(A=@n24, B=@n25), B=@n23, Alpha=@n26)) {texture=/Game/Gadgets/Assets/VinderTech_GliderChute/Glider_Thrive_Female/FX/Texture/skull}
  n28 = Multiply(A=AppendVector(A=@n13.R, B=@n13.R), B=@n23)
  n29 = Constant2Vector() {g=-0.3}
  n30 = TextureSample(UVs=LinearInterpolate(A=Add(A=@n28, B=@n29), B=@n23, Alpha=@n26)) {texture=&skull}
  n31 = Multiply(A=@n19, B=ScalarParameter() {defaultValue=0.3, parameterName=Noise})
  n32 = Add(
   A=LinearInterpolate(A=LinearInterpolate(A=@n27.R, Alpha=@n27.R) {constB=0}, B=@n30.R, Alpha=@n30.R)
   B=TextureSample(UVs=Add(A=@n14, B=@n31)) {texture=&skull}.G
  )
  n33 = Multiply(A=Multiply(A=Multiply(A=@n32) {constB=0.05}, B=@n1), B=ParticleColor().A)
  n34 = MaterialFunctionCall(Float2=TextureCoordinate()) {materialFunction=&BreakOutFloat2Components}
  n35 = ScalarParameter() {defaultValue=9, parameterName=EdgeF}
  n36 = Multiply(A=@n34.G, B=@n35) {constB=5}
  n37 = Multiply(A=OneMinus(None=@n34.G), B=@n35) {constB=5}
  Multiply(A=Multiply(A=Clamp(None=@n4), B=Multiply(A=@n12, B=@n33)), B=Min(A=Saturate(None=@n36), B=Saturate(None=@n37)))
  TextureCoordinate()
  TextureSample(UVs=@n23) {texture=&skull}
 ]
}
