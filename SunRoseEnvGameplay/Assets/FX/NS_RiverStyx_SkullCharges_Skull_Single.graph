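# Niagara system for one River Styx skull charge, a mesh-and-sprite skull hovering by a skeletal mesh with a spawn-in burst.
# User.Index picks one of three offsets; curl noise, bobbing and velocity-scaled smoothing move the skull, which faces its motion.
# Energy sprites and ribbon spawn from the skull, reading DynamicMaterialParameter3 for packed direction and visibility fade.
# The skull also emits location events and exports particle data to Blueprint.
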NiagaraSystem /SunRoseEnvGameplay/Assets/FX/NS_RiverStyx_SkullCharges_Skull_Single
dependencies {
 usedDataInterfaces [
  &NiagaraDataInterfaceCurve
  &NiagaraDataInterfaceArrayFloat3
  &NiagaraDataInterfaceArrayFloat
  &NiagaraDataInterfaceVectorCurve
  /Script/Niagara.NiagaraDataInterfaceExport
  &NiagaraDataInterfaceCamera
  &NiagaraDataInterfaceParticleRead
  &NiagaraDataInterfaceColorCurve
  &NiagaraDataInterfaceSkeletalMesh
 ]
 usedDynamicInputs [
  &ClampFloat
  &RemapRange
  &VectorLength
  &SetBoolByFloatComparison
 ]
 usedModules [
  &self:SystemSpawnScript.SystemScriptSource.SystemScriptGraph.NiagaraNodeAssignment_4.SetVariables_#1
  &SystemState
  &self:SystemSpawnScript.SystemScriptSource.SystemScriptGraph.NiagaraNodeAssignment_2.SetVariables_#2
  &self:SystemSpawnScript.SystemScriptSource.SystemScriptGraph.NiagaraNodeAssignment_1.SetVariables_#3
  &self:SystemSpawnScript.SystemScriptSource.SystemScriptGraph.NiagaraNodeAssignment_0.SetVariables_#4
  &Timeline
  /Game/Effects/Niagara/Modules/Logic/UpdateValueStateChange
  &self:SystemSpawnScript.SystemScriptSource.SystemScriptGraph.NiagaraNodeAssignment_3.SetVariables_#5
  &self:DeactivateSystemOnBool
 ]
}
emitters {
 Energy {
  emitterUpdateScript [
   &EmitterState(
    Enable Distance Culling=true
    Life Cycle Mode=::NewEnumerator1
    Loop Duration=1
    MaxDistance=17000
    Module.bUseMaxDistance=true
    Module.bUseMinDistance=false
    Scalability Mode=::NewEnumerator1
    UseLoopDelay=false
   )
   &SpawnParticlesFromOtherEmitter(
    Attribute Reader=DataInterface() {propertyValues {EmitterBinding {bindingMode=Other, emitterName=Skull_Orbit}}}
    Maximum Age Spawn Threshold=0.8
    Minimum Age Spawn Threshold=0.25
    Spawn Rate=12.5
    Use Maximum Age=true
    Use Minimum Age=true
   )
  ]
  particleSpawnScript [
   &InitializeParticle(
    Color=&MakeCustomLinearColorFromBool(Bool=&RandomBool(), True Color=<rgba 0.0904 0.6198 0.3551 0.2>, False Color=<rgba 0.0145 0.6979 0.3562 0.2>)
    Lifetime=1.5
    Sprite Rotation Mode=::NewEnumerator2
    Sprite Size Mode=::NewEnumerator1
    Uniform Sprite Size Max=15
    Uniform Sprite Size Min=10
    UsePositionOffset=false
    Write Material Random=true
   )
   &SampleParticlesFromOtherEmitter(Dynamic Material Parameter 3 Sampling:ENiagara_DirectReadSamplingMode=::NewEnumerator0)
   &DynamicMaterialParameters(
    Index 0 Param 1=&RandomRangeFloat(Minimum=0, Maximum=1)
    Module.Param0WriteEnabled=true
    Module.Param1WriteEnabled=true
    Module.Param2WriteEnabled=true
    Module.Param3WriteEnabled=true
   ) {module=DynamicMaterialParameters001}
   &ShapeLocation(
    Additional Quaternion Rotation=false
    MaskSpawnGroup=false
    Offset=<-5 0 0>
    Offset Coordinate Space=::Local
    Offset Mode=::NewEnumerator0
    Override Seed=false
    Rotation Mode=::NewEnumerator3
    Rotation Quaternion=&OrientMeshToVector(
     Look At Direction=&NormalizeVector(
      Vector To Normalize=&TransformVector(
       Vector=&MakeVector(
        X=&MakeFloatFromVector4(Vector 4=$Particles.DynamicMaterialParameter3, Channel=::NewEnumerator1)
        Y=&MakeFloatFromVector4(Vector 4=$Particles.DynamicMaterialParameter3, Channel=::NewEnumerator2)
        Z=&MakeFloatFromVector4(Vector 4=$Particles.DynamicMaterialParameter3, Channel=::NewEnumerator3)
       )
       Source Space=::World
       Destination Space=::Simulation
      )
      Safe Normalize=true
      Safe Fallback Threshold=0
      Safe Fallback Vector=<0 0 0.001>
     )
     Reference Axis Vector=<1 0 0>
    )
    Sphere Radius=2.5
    Sphere Surface Distribution=1
    Transform Order:ENiagara_TransformOrder=::NewEnumerator1
    UseOwnerScale=true
   )
   &CameraOffset()
  ]
  particleUpdateScript [
   &ParticleState()
   &ScaleSpriteSize(Scale Factor=&Multiply_Vector2(
     A=<1 1>
     B=&Vector2DFromFloat(Value=&FloatFromCurve(
       FloatCurve=DataInterface() {propertyValues {Curve {keys [
           {arriveTangent=24.2191, interpMode=RCIM_Cubic, leaveTangent=24.2191, tangentMode=RCTM_User, time=0, value=0.5}
           {arriveTangent=1.1179, interpMode=RCIM_Cubic, leaveTangent=1.1179, tangentMode=RCTM_User, time=0.1, value=1.5}
           {arriveTangent=-1.9107, interpMode=RCIM_Cubic, leaveTangent=-1.9107, tangentMode=RCTM_User, time=1, value=0}
          ]}}}
       Scale Curve=&FloatFromCurve(
        FloatCurve=DataInterface() {propertyValues {Curve {keys [
            {interpMode=RCIM_Cubic, tangentMode=RCTM_User, time=0, value=0}
            {arriveTangent=2.0819, interpMode=RCIM_Cubic, leaveTangent=2.0819, tangentMode=RCTM_User, time=0.5, value=1}
           ]}}}
        CurveIndex=$Engine.System.Age
        Scale Curve=0.5
       )
      ))
    ))
   &ScaleColor(
    Linear Color Curve=DataInterface() {propertyValues {
      AlphaCurve {keys [<at 0.0078 0>,<at 0.0872 2>,<at 0.1609 1>,<at 0.4571 1>,<at 1 0>]}
      BlueCurve {keys [<at 0 2>,<at 0.72 0.4555>,<at 0.9893 0.5521>]}
      GreenCurve {keys [<at 0 2>,<at 0.72 0.5521>,<at 0.9893 0.0807>]}
      RedCurve {keys [<at 0 2>,<at 0.72 0.4417>,<at 0.9893 0>]}
     }}
    Scale Mode=::NewEnumerator2
   ) {module=ScaleColor001}
   &Drag(Drag=100, Rotational Drag=5, Use Linear Drag=true, Use Rotational Drag=true)
   &InheritSourceMovement(Applied Position Scale=<0.9 0.9 0.9>, ApplyForce=false, ApplyPosition=true, ApplyPositionFalloff=false)
   &SolveForcesAndVelocity(Clamp Velocity=false, Limit Acceleration=false)
   &ScaleColor(
    Scale Alpha=&OneMinusFloat(Float=&MakeFloatFromVector4(Vector 4=$Particles.DynamicMaterialParameter3, Channel=::NewEnumerator0))
    ScaleA=true
    ScaleRGB=false
   )
  ]
  rendererClasses [&NiagaraSpriteRendererProperties]
  renderers [{rendererClass=&NiagaraSpriteRendererProperties}]
 }
 Ribbon_Attached_to_Skull {
  emitterUpdateScript [
   &EmitterState(Life Cycle Mode=::NewEnumerator1, Loop Duration=1, UseLoopDelay=false)
   &SpawnParticlesFromOtherEmitter(
    Attribute Reader=DataInterface() {propertyValues {EmitterBinding {bindingMode=Other, emitterName=Skull_Orbit}}}
    Spawn Rate=50
    Spawn Rate Per Particle Cap=250
    Use Maximum Age=false
    Use Minimum Age=false
   )
  ]
  particleSpawnScript [
   &InitializeParticle(UsePositionOffset=false, Write Material Random=true)
   &InitializeRibbon(
    Color=<rgba 1 1 1 0.75>
    Lifetime=0.25
    Position Mode:ENiagara_PositionInitializationMode=::NewEnumerator0
    Ribbon Facing Mode:ENiagara_UnsetDirectSet=::NewEnumerator1
    Ribbon Facing Vector:Vector3f=<1 0 0>
    Ribbon Width=1
    Write Material Random=true
   )
   &SampleParticlesFromOtherEmitter(Dynamic Material Parameter 3 Sampling:ENiagara_DirectReadSamplingMode=::NewEnumerator0, Particle ID Sampling=::NewEnumerator1)
   /SetVariables(Particles.RibbonWidth=5) {isSetParameters=true}
   &AddVelocity(
    Additional Quaternion Rotation=false
    Rotation Mode=::NewEnumerator3
    Rotation Quaternion:Quat4f=&OrientMeshToVector(
     Look At Direction=&NormalizeVector(
      Vector To Normalize=&TransformVector(
       Vector=&MakeVector(
        X=&MakeFloatFromVector4(Vector 4=$Particles.DynamicMaterialParameter3, Channel=::NewEnumerator1)
        Y=&MakeFloatFromVector4(Vector 4=$Particles.DynamicMaterialParameter3, Channel=::NewEnumerator2)
        Z=&MakeFloatFromVector4(Vector 4=$Particles.DynamicMaterialParameter3, Channel=::NewEnumerator3)
       )
       Source Space=::World
       Destination Space=::Simulation
      )
      Safe Normalize=true
      Safe Fallback Threshold=0
      Safe Fallback Vector=<0 0 0.001>
     )
     Reference Axis Vector=<1 0 0>
    )
    Velocity=<-1 0 0>
    Velocity Speed Scale=25
   )
   &CameraOffset(Camera Offset Amount=20)
  ]
  particleUpdateScript [
   &ParticleState()
   &InheritSourceMovement(Applied Position Scale=<0.8 0.8 1>, ApplyForce=false, ApplyPosition=true, ApplyPositionFalloff=false) {module=InheritSourceMovement001}
   &SolveForcesAndVelocity(Clamp Velocity=false, Limit Acceleration=false)
   &ScaleColor(
    Linear Color Curve=DataInterface() {propertyValues {
      AlphaCurve {keys [<at 0.0062 0>,<at 0.1273 1>,<at 0.6974 0.75>,<at 1 0>]}
      BlueCurve {keys [<at 0 0.1667>,<at 0.5816 0.0466>,<at 0.8584 0.5839>]}
      GreenCurve {keys [<at 0 0.5>,<at 0.5816 0.4>,<at 0.8584 0.7>]}
      RedCurve {keys [<at 0 0>,<at 0.5816 0>,<at 0.8584 0>]}
     }}
    Scale Mode=::NewEnumerator2
   ) {module=ScaleColor001}
   &DynamicMaterialParameters(
    Index 1 Param 4=1
    Module.Index1Param0Write=false
    Module.Index1Param1Write=false
    Module.Index1Param2Write=false
    Module.Index1Param3Write=true
    Write Parameter Index 0=false
    Write Parameter Index 1=true
   )
   &ScaleColor(
    Scale Alpha=&OneMinusFloat(Float=&MakeFloatFromVector4(Vector 4=$Particles.DynamicMaterialParameter3, Channel=::NewEnumerator0))
    ScaleA=true
    ScaleRGB=false
   )
   &ScaleColor(Scale Alpha=0.5, ScaleA=true, ScaleRGB=false) {module=ScaleColor002}
   &ScaleRibbonWidth(Ribbon Width Scale=&FloatFromCurve(FloatCurve=DataInterface() {propertyValues {Curve {keys [
         {interpMode=RCIM_Cubic, tangentMode=RCTM_User, time=0, value=1}
         {arriveTangent=-0.073, interpMode=RCIM_Cubic, leaveTangent=-0.073, tangentMode=RCTM_User, time=1, value=0.75}
        ]}}}))
  ]
  rendererClasses [&NiagaraRibbonRendererProperties]
  renderers [{rendererClass=&NiagaraRibbonRendererProperties}]
 }
 Skull_Orbit {
  emitterUpdateScript [
   &EmitterState(
    Enable Distance Culling=true
    Life Cycle Mode=::NewEnumerator1
    Loop Behavior=::NewEnumerator1
    Loop Duration Mode=::NewEnumerator1
    MaxDistance=15000
    Module.bApplySpawnCountScaleByDistance=false
    Module.bUseMaxDistance=true
    Module.bUseMinDistance=true
    Scalability Mode=::NewEnumerator1
    Scale Spawn Count=true
    Spawn Count Scale By Distance=DataInterface() {propertyValues {Curve {keys [<at 0 0>,<at 0.2 1>,<at 0.85 1>,<at 1 0>]}}}
    UseLoopDelay=false
   )
   &SpawnBurst_Instantaneous(Use Spawn Probability=false, UseLoopCountLimit=false)
  ]
  particleSpawnScript [
   &InitializeParticle(
    Color=<rgba 0 0.625 1.5 2>
    Lifetime=2
    Mesh Scale Mode=::NewEnumerator3
    Mesh Uniform Scale=0.5
    Sprite Size=<28 20>
    Sprite Size Mode=::NewEnumerator4
    UsePositionOffset=false
    Write Material Random=true
   )
   &InitialMeshOrientation(Mesh Axis To Orient=<0 1 0>, Mesh Orientation Mode=::NewEnumerator2, Use Rotation Vector=false)
   &CameraOffset(Camera Offset Amount=&RandomRangeFloat(Minimum=25, Maximum=50))
   /SetVariables(
    Particles.Color_Sprite_Dark:LinearColor=<rgba 0 0 0 2>
    Particles.Sprite_Size_Dark:Vector2f=&Multiply_Vector2D_ByFloat(Float=0.5, Vector2D=<20 20>)
   ) {isSetParameters=true}
   /SetVariables(
    Particles.Color_Sprite_Light:LinearColor=<rgba 0 1.5 1 0.25>
    Particles.Sprite_Size_Light:Vector2f=&Multiply_Vector2D_ByFloat(Float=0.5, Vector2D=<20 20>)
   ) {isSetParameters=true}
   /SetVariables(Particles.DynamicParametersSprite:Vector4f=<25 0 0 0>, Particles.DynamicParametersSpriteDark:Vector4f=<0 1 0 0>) {isSetParameters=true}
   &ApplyInitialForces()
  ]
  particleUpdateScript [
   &ParticleState(Kill Particles When Lifetime Has Elapsed=false, Let Infinitely Lived Particles Die When Emitter Deactivates=true)
   &SkeletalMeshLocation(
    MaskSpawnGroup=false
    Mesh Sampling Type=::NewEnumerator5
    Orientation Sampling=::NewEnumerator1
    Position Sampling=::NewEnumerator2
    Sampled Position Offset=<0 0 100>
    Sampled Position Offset Coordinate Space:ENiagaraSkelMeshTransforms=::NewEnumerator2
    Skeletal Mesh=$System.Skeletal Mesh
    Socket Sampling Mode=::NewEnumerator5
   )
   &CurlNoiseForce(
    Noise Frequency=1
    Noise Quality / Cost=::NewEnumerator0
    Noise Strength=5
    Pan Noise=true
    Pan Noise Field=&VectorFromFloat(Value=0.1)
    Random Seed:NiagaraInt32=$User.Index
    Randomization Vector:Vector3f=&SelectVectorFromArray(
     Direct Array Index=$User.Index
     Vector Selection Array=DataInterface() {propertyValues {FloatData [<0.65 0.125 0.37>, <0.254 0.842 0.178>, <0.1353 0.524 0.777>]}}
    )
    Randomize Noise Sample=false
    Write to Intrinsic Parameters=false
   )
   /Niagara/Noise/Waveform(
    EnableExponent1=false
    GlobalClampMax=false
    GlobalClampMin=false
    UseInput1=false
    UsePhase1=false
    "Waveform [1]":ENiagara_Waveforms=::NewEnumerator12
    "[1] Frequency"=2
   ) {module=Waveform003}
   /SetVariables(Particles.Position=&AddVectorToPosition(
     Position=$Particles.SkeletalMeshLocation.SampledPosition
     Vector=&TransformVector(
      Vector=&AddVectorToPosition(
       Position=&SelectVectorFromArray(Direct Array Index=$User.Index, Vector Selection Array=DataInterface() {propertyValues {FloatData [<-30 25 -5>, <-50 0 25>, <-40 -25 5>]}})
       Vector=&Multiply_Vector(
        A=$Output.CurlNoiseForce.CurlNoiseForce
        B=&MakeVector(X=1, Y=1, Z=&Multiply_Float(A=$Particles.Waveform003.Waveform, B=2))
       )
      )
      Destination Space=::Simulation
     )
    )) {isSetParameters=true}
   &PendulumSetup(
    Module.Enable Spring Driver=false
    Pendulum Length=2.5
    Pendulum Pivot:Vector3f=&AddVectorToPosition(
     Position=$Particles.SkeletalMeshLocation.SampledPosition
     Vector=&TransformVector(
      Vector=&SelectVectorFromArray(Direct Array Index=$User.Index, Vector Selection Array=DataInterface() {propertyValues {FloatData [<-30 25 -5>, <-50 0 25>, <-40 -25 5>]}})
      Destination Space=::Simulation
     )
    )
    Pendulum Pivot Offset:Vector3f=/Niagara/DynamicInputs/Lerp/SmoothLerpOverTime_Vector(
     Convergence Rate:NiagaraFloat=1
     Smooth Value:Vector3f=$Output.CurlNoiseForce.CurlNoiseForce
     Target Value:Vector3f=$Particles.LastCurlNoiseFOrce
    )
    Pendulum Pivot Offset Coordinate Space:ENiagaraCoordinateSpace=::Local
    Rigid Pendulum=false
    "Spring Driven Constraint (Tightness)"=5
   ) {enabled=false}
   /SetVariables(Particles.Position_Before:NiagaraPosition=$Particles.Position) {isSetParameters=true}
   &InterpolateOverTime(
    Initial Vector Value:NiagaraPosition=$Particles.Position_Before
    Rate Of Change:NiagaraFloat=&FloatFromCurve(
     FloatCurve=DataInterface() {propertyValues {Curve {keys [<at 0 4 RCIM_Cubic>, {arriveTangent=31.6088, interpMode=RCIM_Cubic, leaveTangent=31.6088, tangentMode=RCTM_User, time=1, value=16}]}}}
     CurveIndex=$System.NormalizedVelocityRange
     Scale Curve=&SelectFloatFromArray(Direct Array Index=$User.Index, Float Selection Array=DataInterface() {propertyValues {FloatData [3.5,3,4]}})
    )
    Target Vector Value:NiagaraPosition=$Particles.Position
    Variable Type=::NewEnumerator2
   ) {module=InterpolateOverTime002}
   /SetVariables(Particles.Position=$Particles.InterpolateOverTime002.InterpolateOverTime_V3_.Moving Average) {isSetParameters=true}
   /SetVariables(Particles.LookAtDirectionBefore:Vector3f=&NormalizeVector(
     Vector To Normalize=&Subtract_Position(
      A=&TransformPosition(Position=$Particles.Position, Source Space=::Local, Destination Space=::Local)
      B=&TransformPosition(Position=$Particles.Previous.Position, Source Space=::Local, Destination Space=::Local)
     )
     Safe Normalize=true
     Safe Fallback Threshold=0
     Safe Fallback Vector=<0.0001 0.0001 0.0001>
    )) {isSetParameters=true}
   &InterpolateOverTime(
    Initial Vector Value:Vector3f=$Particles.LookAtDirectionBefore
    Rate Of Change:NiagaraFloat=&SelectFloatFromArray(Direct Array Index=$User.Index, Float Selection Array=DataInterface() {propertyValues {FloatData [0.5,0.75,1]}})
    Target Vector Value:Vector3f=&NormalizeVector(
     Vector To Normalize=&Subtract_Position(
      A=&TransformPosition(Position=$Particles.Position, Source Space=::Local, Destination Space=::Local)
      B=&TransformPosition(Position=$Particles.Previous.Position, Source Space=::Local, Destination Space=::Local)
     )
     Safe Normalize=true
     Safe Fallback Threshold=0
     Safe Fallback Vector=<0.0001 0.0001 0.0001>
    )
    Variable Type=::NewEnumerator2
   )
   /SetVariables(Particles.LookAtDirection:Vector3f=$StackContext.InterpolateOverTime.InterpolateOverTime_V3_.Moving Average) {isSetParameters=true}
   /Niagara/Modules/Update/Orientation/V2/OrientMeshToVector(
    Look At Direction:Vector3f=$Particles.LookAtDirection
    Override Quat=false
    Up Axis Direction:Vector3f=<0 0 1>
    UseOptionalSideAxis=false
    UseOptionalUpAxis=true
   ) {module=OrientMeshToVector003}
   &InterpolateOverTime(
    Initial Vector Value:Vector3f=$Particles.Previous.Velocity
    Rate Of Change=4
    Target Vector Value:Vector3f=$StackContext.Velocity
    Variable Type=::NewEnumerator2
   ) {enabled=false, module=InterpolateOverTime001}
   &OrientMeshToVector(
    Look At Direction:Vector3f=&NormalizeVector(
     Vector To Normalize=$Particles.InterpolateOverTime001.InterpolateOverTime_V3_.Moving Average
     Safe Normalize=true
     Safe Fallback Threshold=0
     Safe Fallback Vector=<0 0 0.001>
    )
    Override Quat=false
    Up Axis Direction:Vector3f=<0 0 1>
    UseOptionalSideAxis=false
    UseOptionalUpAxis=true
   ) {enabled=false}
   /SetVariables(
    Particles.Color_Sprite_Dark:LinearColor=&MultiplyLinearColor(
     Linear Color=$Particles.Initial.Color_Sprite_Dark
     Scale Factor=&MakeFloat4FromVectorAndFloat(
      "Vector (RGB)"=<1 1 1>
      "Float (Alpha)"=&FloatFromCurve(FloatCurve=DataInterface() {propertyValues {Curve {keys [
           {interpMode=RCIM_Cubic, tangentMode=RCTM_User, time=0, value=0.75}
           {interpMode=RCIM_Cubic, tangentMode=RCTM_User, time=0.5, value=1}
           {tangentMode=RCTM_User, time=1, value=1}
          ]}}})
     )
    )
    Particles.Sprite_Size_Dark:Vector2f=&Multiply_Vector2D_ByFloat(
     Float=&Multiply_Float(
      A=&FloatFromCurve(
       FloatCurve=DataInterface() {propertyValues {Curve {keys [
           {interpMode=RCIM_Cubic, tangentMode=RCTM_User, time=0, value=0}
           {arriveTangent=1.9474, interpMode=RCIM_Cubic, leaveTangent=1.9474, tangentMode=RCTM_User, time=1, value=1}
          ]}}}
       CurveIndex=$Engine.System.Age
      )
      B=&FloatFromCurve(FloatCurve=DataInterface() {propertyValues {Curve {keys [
           {interpMode=RCIM_Cubic, tangentMode=RCTM_User, time=0, value=0.75}
           {interpMode=RCIM_Cubic, tangentMode=RCTM_User, time=0.5, value=1}
           {tangentMode=RCTM_User, time=1, value=1}
          ]}}})
     )
     Vector2D=$Particles.Initial.Sprite_Size_Dark
    )
   ) {isSetParameters=true}
   /SetVariables(
    Particles.Color_Sprite_Light:LinearColor=&MultiplyLinearColor(
     Linear Color=$Particles.Initial.Color_Sprite_Light
     Scale Factor=&MakeFloat4FromVectorAndFloat(
      "Vector (RGB)"=<1 1 1>
      "Float (Alpha)"=&FloatFromCurve(FloatCurve=DataInterface() {propertyValues {Curve {keys [
           {interpMode=RCIM_Cubic, tangentMode=RCTM_User, time=0, value=0.75}
           {interpMode=RCIM_Cubic, tangentMode=RCTM_User, time=0.5, value=1}
           {tangentMode=RCTM_User, time=1, value=1}
          ]}}})
     )
    )
    Particles.Sprite_Size_Light:Vector2f=&Multiply_Vector2D_ByFloat(
     Float=&Multiply_Float(
      A=&FloatFromCurve(
       FloatCurve=DataInterface() {propertyValues {Curve {keys [
           {interpMode=RCIM_Cubic, tangentMode=RCTM_User, time=0, value=0}
           {arriveTangent=1.9474, interpMode=RCIM_Cubic, leaveTangent=1.9474, tangentMode=RCTM_User, time=1, value=1}
          ]}}}
       CurveIndex=$Engine.System.Age
      )
      B=&FloatFromCurve(FloatCurve=DataInterface() {propertyValues {Curve {keys [
           {interpMode=RCIM_Cubic, tangentMode=RCTM_User, time=0, value=0.75}
           {interpMode=RCIM_Cubic, tangentMode=RCTM_User, time=0.5, value=1}
           {tangentMode=RCTM_User, time=1, value=1}
          ]}}})
     )
     Vector2D=$Particles.Initial.Sprite_Size_Light
    )
   ) {isSetParameters=true}
   &ScaleMeshSize(Scale Factor=&VectorFromCurve(
     VectorCurve=DataInterface() {propertyValues {
       XCurve {keys [
         {interpMode=RCIM_Cubic, tangentMode=RCTM_User, time=0, value=0}
         {arriveTangent=2.343, interpMode=RCIM_Cubic, leaveTangent=2.343, tangentMode=RCTM_User, time=1, value=1}
        ]}
       YCurve {keys [
         {interpMode=RCIM_Cubic, tangentMode=RCTM_User, time=0, value=0}
         {arriveTangent=2.343, interpMode=RCIM_Cubic, leaveTangent=2.343, tangentMode=RCTM_User, time=1, value=1}
        ]}
       ZCurve {keys [
         {interpMode=RCIM_Cubic, tangentMode=RCTM_User, time=0, value=0}
         {arriveTangent=2.343, interpMode=RCIM_Cubic, leaveTangent=2.343, tangentMode=RCTM_User, time=1, value=1}
        ]}
      }}
     CurveIndex=$System.Age
     Scale Curve=<0.7 0.7 0.7>
    ))
   &SolveForcesAndVelocity(Clamp Velocity=false, Limit Acceleration=false)
   &GenerateLocationEvent(
    Event Generation Enabled:NiagaraBool=&SetBoolByIntComparison(A=$System.LoopCount, B=1, Comparison Type=::NewEnumerator6)
    Event Send Rate=2
    Use Event Probability=false
    UseEventDelay=false
   )
   /SetVariables(Particles.VisibilityTag=&MakeCustomIntFromBool(
     Bool=&Boolean_AndOperation(
      Bool1=&SetBoolByIntComparison(A=$System.UpdateValueStateChange.IntParam, B=1, Comparison Type=::NewEnumerator0)
      Bool2=$System.UpdateValueStateChange.Changed
     )
     True INT=1
     False INT=0
    )) {isSetParameters=true}
   &PendulumConstraint() {enabled=false}
   &CalculateAccurateVelocity() {enabled=false}
   /Niagara/Modules/ExportParticleData/ExportParticleDataToBlueprint(
    Condition To Export Data=true
    Export Particle Data Interface=DataInterface() {propertyValues {
      GPUAllocationFixedSize=64
      GPUAllocationMode=FixedSize
      GPUAllocationPerParticleSize=1
      CallbackHandlerParameter {parameter {name=User.BPCallback, typeDefHandle {registeredTypeIndex=1}, varData []}}
     }}
    UseDelayTimer=false
   )
   &InterpolateOverTime(
    Rate Of Change:NiagaraFloat=&FloatFromCurve(
     FloatCurve=DataInterface() {propertyValues {Curve {keys [
         {arriveTangent=3.7228, interpMode=RCIM_Cubic, leaveTangent=3.7228, tangentMode=RCTM_User, time=0, value=1}
         {arriveTangent=-0.3153, interpMode=RCIM_Cubic, leaveTangent=-0.3153, tangentMode=RCTM_User, time=1, value=2.5}
        ]}}}
     CurveIndex=$System.NormalizedVelocityRange
     Scale Curve=&SelectFloatFromArray(Direct Array Index=$User.Index, Float Selection Array=DataInterface() {propertyValues {FloatData [3,4,3.5]}})
    )
    Target Quat Value:Quat4f=$Particles.MeshOrientation
    Variable Type=::NewEnumerator5
   ) {module=InterpolateOverTime003}
   /SetVariables(Particles.SmoothedOrientation:Quat4f=$StackContext.InterpolateOverTime003.InterpolateOverTime_Quaternion.Moving Average) {isSetParameters=true}
   &DynamicMaterialParameters(
    Index 0 Param 1=1
    Index 1 Param 1=&RandomRangeFloat(Minimum=0.2, Maximum=0.2)
    Index 3 Param 1=&MakeFloatFromInt(Int=$Particles.VisibilityTag)
    Index 3 Param 2=&MakeFloatFromVector(Vector=$Particles.InterpolateOverTime.InterpolateOverTime_V3_.Moving Average, Channel=::NewEnumerator0)
    Index 3 Param 3=&MakeFloatFromVector(Vector=$Particles.InterpolateOverTime.InterpolateOverTime_V3_.Moving Average, Channel=::NewEnumerator1)
    Index 3 Param 4=&MakeFloatFromVector(Vector=$Particles.InterpolateOverTime.InterpolateOverTime_V3_.Moving Average, Channel=::NewEnumerator2)
    Module.Index0Param0Write=true
    Module.Index0Param1Write=true
    Module.Index0Param2Write=true
    Module.Index0Param3Write=true
    Module.Index1Param0Write=true
    Module.Index1Param1Write=true
    Module.Index1Param2Write=true
    Module.Index1Param3Write=true
    Module.Index2Param0Write=true
    Module.Index2Param1Write=true
    Module.Index2Param2Write=true
    Module.Index2Param3Write=false
    Module.Index3Param0Write=true
    Module.Index3Param1Write=true
    Module.Index3Param2Write=true
    Module.Index3Param3Write=true
    Write Parameter Index 1=true
    Write Parameter Index 2=true
    Write Parameter Index 3=true
   )
   /SetVariables(Particles.SmoothedOrientation:Quat4f=&MultiplyQuaternion(
     Quaternion A=&MultiplyQuaternion(
      Quaternion A=&MakeQuatFromAxisAngle(
       Axis=<0 0 1>
       AngleInDegrees=&Multiply_Float(A=&Multiply_Float(A=$System.Age, B=0.25), B=180)
      )
      Quaternion B=&MakeQuatFromAxisAngle(
       Axis=<1 0 0>
       AngleInDegrees=/Niagara/DynamicInputs/Helpers/Waveform(
        Global Phase=&Multiply_Float(A=$Emitter.Age, B=0.1)
        Global Amplitude Scale=35
        "Waveform [1]"=::NewEnumerator0
        "[1] Frequency"=1
        "[1] Amplitude Min/Max"=<-1 1>
       )
      )
     )
     Quaternion B=/Niagara/DynamicInputs/Vectors/OrientMeshToVector(
      Look At Direction=&ReturnCameraVector(Camera Vectors=::NewEnumerator1, Coordinate Space=::Simulation)
      Reference Axis Vector=<1 0 0>
     )
    )) {enabled=false, isSetParameters=true}
   /SetVariables(Particles.SmoothedOrientation:Quat4f=&MultiplyQuaternion(
     Quaternion A=&MultiplyQuaternion(
      Quaternion A=&OrientMeshToVector(
       Look At Direction=&ReturnCameraVector(Camera Vectors=::NewEnumerator4, Coordinate Space=::Simulation, Origin Position=$Particles.Position)
       Reference Axis Vector=<0 1 0>
      )
      Quaternion B=&MakeQuatFromAxisAngle(Axis=<0 0 1>, AngleInDegrees=90)
     )
     Quaternion B=&MultiplyQuaternion(
      Quaternion A=&MakeQuatFromAxisAngle(
       Axis=<0 0 1>
       AngleInDegrees=&Waveform(
        Global Phase=&Multiply_Float(A=$System.Age, B=0.1)
        Global Amplitude Scale=45
        "Waveform [1]"=::NewEnumerator0
        "[1] Frequency"=1
        "[1] Amplitude Min/Max"=<-1 1>
       )
      )
      Quaternion B=&MakeQuatFromAxisAngle(
       Axis=<1 0 0>
       AngleInDegrees=&Waveform(
        Global Phase=&Multiply_Float(A=$System.Age, B=0.5)
        Global Amplitude Scale=45
        "Waveform [1]"=::NewEnumerator12
        "[1] Frequency"=1
        "[1] Amplitude Min/Max"=<-1 1>
       )
      )
     )
    )) {enabled=false, isSetParameters=true}
  ]
  rendererClasses [&NiagaraMeshRendererProperties, &NiagaraSpriteRendererProperties]
  renderers [
   {rendererClass=&NiagaraMeshRendererProperties, rendererIndex=0}
   {rendererClass=&NiagaraSpriteRendererProperties, rendererIndex=1}
   {rendererClass=&NiagaraSpriteRendererProperties, rendererIndex=2}
  ]
 }
 SpawnIn {
  emitterUpdateScript [&EmitterState(Life Cycle Mode=::NewEnumerator1, Loop Duration=1, UseLoopDelay=false)]
  particleSpawnScript [
   &InitializeParticle(
    Color=<rgba 0 0.5469 0.2734 0.75>
    Lifetime=0.35
    Mesh Scale Mode=::NewEnumerator3
    Sprite Size Mode=::NewEnumerator3
    Uniform Sprite Size=&Multiply_Float(A=150, B=$User.Burst_Size)
    UsePositionOffset=false
    Write Material Random=true
   )
   &CameraOffset(Camera Offset Amount=&MaskFloatBySpawnGroup(
     Masked Float=30
     Passthrough Float=&MaskFloatBySpawnGroup(Masked Float=50, Passthrough Float=50, Spawn Group Mask=1)
    ))
   /SetVariables(
    Particles.Color_Sprite_Dark:LinearColor=<rgba 0 0 0 0.5>
    Particles.Sprite_Size_Dark:Vector2f=&Multiply_Vector2D_ByFloat(Float=$User.Burst_Size, Vector2D=<500 500>)
   ) {isSetParameters=true}
   /SetVariables(
    Particles.Mesh_Color_EnergyGreen:LinearColor=<rgba 0 1 0.6667 0.75>
    Particles.Mesh_Size_EnergyGreen:Vector3f=&VectorFromFloat(Value=&Multiply_Float(A=0.5, B=$User.Burst_Size))
   ) {isSetParameters=true}
   /SetVariables(
    Particles.Mesh_Color_EnergyDark:LinearColor=<rgba 0 0.2152 1 0.75>
    Particles.Mesh_Size_EnergyDark:Vector3f=&VectorFromFloat(Value=&Multiply_Float(A=0.5, B=$User.Burst_Size))
   ) {isSetParameters=true}
   &InitialMeshOrientation(
    Mesh Axis To Orient=<0 0 1>
    Mesh Orientation Mode=::NewEnumerator2
    Orientation Vector=<0 0 1>
    Rotation=<0 0.25 0>
    Use Rotation Vector=true
   )
  ]
  particleUpdateScript [
   &ParticleState()
   &ScaleSpriteSize(Scale Factor=&Vector2DFromFloat(Value=&Multiply_Float(
      A=&RandomRangeFloat(Minimum=0.5, Maximum=1, Evaluation Type=::NewEnumerator1)
      B=&FloatFromCurve(FloatCurve=DataInterface() {propertyValues {Curve {keys [<at 0 0>,<at 0.15 1>,<at 1 0>]}}})
     )))
   /SetVariables(
    Particles.DynamicParameter_EnergyDark:Vector4f=&MakeVector4(
     X=&FloatFromCurve(FloatCurve=DataInterface() {propertyValues {Curve {keys [
          {arriveTangent=3.4805, interpMode=RCIM_Cubic, leaveTangent=3.4805, tangentMode=RCTM_User, time=0, value=-0.83}
          {arriveTangent=0.5266, interpMode=RCIM_Cubic, leaveTangent=0.5266, tangentMode=RCTM_User, time=1, value=0.25}
         ]}}})
     Y=1
     Z=1
     W=1
    )
    Particles.Mesh_Color_EnergyDark:LinearColor=&MultiplyLinearColor(
     Linear Color=$Particles.Initial.Mesh_Color_EnergyDark
     Scale Factor=&MakeFloat4FromVectorAndFloat(
      "Vector (RGB)"=&Multiply_VectorByFloat(
       Vector=<2 3 1>
       Float=&FloatFromCurve(
        FloatCurve=DataInterface() {propertyValues {Curve {keys [
            {interpMode=RCIM_Cubic, tangentMode=RCTM_User, time=0, value=0}
            {arriveTangent=2.2532, interpMode=RCIM_Cubic, leaveTangent=2.2532, tangentMode=RCTM_User, time=0.3029, value=0.9805}
            {interpMode=RCIM_Cubic, tangentMode=RCTM_User, time=1, value=0}
           ]}}}
        Scale Curve=0.75
       )
      )
      "Float (Alpha)"=&Multiply_Float(
       A=&FNTimeOfDayScalar(Night Scale=0.3, Morning Scale=0.3, Evening Scale=0.5)
       B=&FloatFromCurve(
        FloatCurve=DataInterface() {propertyValues {Curve {keys [
            {arriveTangent=0.2654, interpMode=RCIM_Cubic, leaveTangent=0.2654, tangentMode=RCTM_User, time=0, value=0}
            {arriveTangent=1.1179, interpMode=RCIM_Cubic, leaveTangent=1.1179, tangentMode=RCTM_User, time=0.2409, value=0.5646}
            {arriveTangent=0.0224, interpMode=RCIM_Cubic, leaveTangent=0.0224, tangentMode=RCTM_User, time=1, value=1}
           ]}}}
        Scale Curve=&FloatFromCurve(FloatCurve=DataInterface() {propertyValues {Curve {keys [<at 0 0.5>,<at 0.2 1>]}}}, Scale Curve=0.1)
       )
      )
     )
    )
    Particles.Mesh_Size_EnergyDark:Vector3f=&Multiply_Vector(
     A=$Particles.Initial.Mesh_Size_EnergyDark
     B=&VectorFromFloat(Value=&FloatFromCurve(FloatCurve=DataInterface() {propertyValues {Curve {keys [{interpMode=RCIM_Cubic, tangentMode=RCTM_User, time=0, value=1}, {interpMode=RCIM_Cubic, tangentMode=RCTM_User, time=1, value=0}]}}}))
    )
   ) {isSetParameters=true}
   /SetVariables(
    Particles.Mesh_Color_EnergyGreen:LinearColor=&MultiplyLinearColor(
     Linear Color=$Particles.Initial.Mesh_Color_EnergyGreen
     Scale Factor=&MakeFloat4FromVectorAndFloat(
      "Vector (RGB)"=<1 1 1>
      "Float (Alpha)"=&Multiply_Float(
       A=&FloatFromCurve(FloatCurve=DataInterface() {propertyValues {Curve {keys [
            {arriveTangent=9.0725, interpMode=RCIM_Cubic, leaveTangent=9.0725, tangentMode=RCTM_User, time=0, value=0}
            {arriveTangent=-3.668, interpMode=RCIM_Cubic, leaveTangent=-3.668, tangentMode=RCTM_User, time=0.1408, value=0.9847}
            {arriveTangent=-0.0405, interpMode=RCIM_Cubic, leaveTangent=-0.0405, tangentMode=RCTM_User, time=1, value=0}
           ]}}})
       B=&FNTimeOfDayScalar(Night Scale=0.7, Morning Scale=0.5, Evening Scale=0.5)
      )
     )
    )
    Particles.Mesh_Size_EnergyGreen:Vector3f=&Multiply_Vector(
     A=$Particles.Initial.Mesh_Size_EnergyGreen
     B=&VectorFromFloat(Value=&FloatFromCurve(FloatCurve=DataInterface() {propertyValues {Curve {keys [
           {arriveTangent=4.1785, interpMode=RCIM_Cubic, leaveTangent=4.1785, tangentMode=RCTM_User, time=0, value=0}
           {arriveTangent=1.1179, interpMode=RCIM_Cubic, leaveTangent=1.1179, tangentMode=RCTM_User, time=0.2042, value=0.6228}
           {arriveTangent=0.0224, interpMode=RCIM_Cubic, leaveTangent=0.0224, tangentMode=RCTM_User, time=1, value=1}
          ]}}}))
    )
   ) {isSetParameters=true}
   /SetVariables(
    Particles.Color_Sprite_Dark:LinearColor=&MultiplyLinearColor(
     Linear Color=$Particles.Initial.Color_Sprite_Dark
     Scale Factor=&MakeFloat4FromVectorAndFloat(
      "Vector (RGB)"=<1 1 1>
      "Float (Alpha)"=&FloatFromCurve(FloatCurve=DataInterface() {propertyValues {Curve {keys [
           {interpMode=RCIM_Cubic, tangentMode=RCTM_User, time=0, value=0.75}
           {interpMode=RCIM_Cubic, tangentMode=RCTM_User, time=0.5, value=1}
           {tangentMode=RCTM_User, time=1, value=0.75}
          ]}}})
     )
    )
    Particles.Sprite_Size_Dark:Vector2f=&Multiply_Vector2D_ByFloat(
     Float=&FloatFromCurve(
      FloatCurve=DataInterface() {propertyValues {Curve {keys [
          {interpMode=RCIM_Cubic, tangentMode=RCTM_User, time=0, value=0.75}
          {interpMode=RCIM_Cubic, tangentMode=RCTM_User, time=0.5, value=1}
          {tangentMode=RCTM_User, time=1, value=0.75}
         ]}}}
      Scale Curve=1.25
     )
     Vector2D=$Particles.Initial.Sprite_Size_Dark
    )
   ) {isSetParameters=true}
   &ScaleColor(
    Scale Alpha=&FloatFromCurve(
     FloatCurve=DataInterface() {propertyValues {Curve {keys [
         {interpMode=RCIM_Cubic, tangentMode=RCTM_User, time=0, value=0}
         {interpMode=RCIM_Cubic, tangentMode=RCTM_User, time=0.4, value=1}
         {tangentMode=RCTM_User, time=0.6, value=0}
        ]}}}
     Scale Curve=5
    )
    ScaleA=true
    ScaleRGB=true
   )
   &MeshRotationForce(Rotation=&MakeVector(
     X=&FloatFromCurve(FloatCurve=DataInterface() {propertyValues {Curve {keys [<at 0 1>,<at 1 0>]}}})
     Y=0
     Z=0
    ))
   &SolveForcesAndVelocity(Clamp Velocity=false, Limit Acceleration=false)
  ]
  rendererClasses [&NiagaraSpriteRendererProperties, &NiagaraMeshRendererProperties]
  renderers [
   {rendererClass=&NiagaraSpriteRendererProperties}
   {rendererClass=&NiagaraSpriteRendererProperties, rendererIndex=1}
   {rendererClass=&NiagaraMeshRendererProperties, rendererIndex=2}
   {rendererClass=&NiagaraMeshRendererProperties, rendererIndex=3}
  ]
 }
}
summary {
 systemName=NS_RiverStyx_SkullCharges_Skull_Single
 emitters [
  {emitterName=SpawnIn, rendererClasses [&NiagaraSpriteRendererProperties, &NiagaraMeshRendererProperties]}
  {emitterName=Skull_Orbit, rendererClasses [&NiagaraMeshRendererProperties, &NiagaraSpriteRendererProperties]}
  {emitterName=Ribbon_Attached_to_Skull, rendererClasses [&NiagaraRibbonRendererProperties]}
  {emitterName=Energy, rendererClasses [&NiagaraSpriteRendererProperties]}
 ]
 userVariables [
  User.bIsFrontEnd {type=NiagaraBool}
  User.bIsFrontEndPreview {type=NiagaraBool}
  User.BPCallback {
   type=Object
   description=""
   defaultValue {
    struct=NiagaraExt_VariableValue_Object
    value {object=/SunRoseEnvGameplay/GameplayCues/GCNL_SunRose_SilkyWolf_Ready, objectClass=&Object}
   }
  }
  User.Burst_Size {type=NiagaraFloat, defaultValue=0.2, description=""}
  User.Index {type=NiagaraInt32}
  User.isDead {type=NiagaraBool}
  User.ParaGlideLeanAlpha {type=NiagaraFloat, defaultValue=0.5, description=""}
  User.Skull_Charge_Counter {type=NiagaraInt32}
  User.State {type=NiagaraInt32}
 ]
}
